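; Manager component that ducks audio through a control bus mix while one or more pawns are registered as emoting.
; Once the game state is ready, it listens on the event router for pawn emote registration events, whose flag selects register or unregister.
; It keeps a unique list of emoting pawns and updates the count from it.
; The ducker mix activates when the count is above zero and deactivates when it returns to zero.
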(event EventBeginPlay
  (Class|FortAsyncActionGameStateReady|GameStateReadyAsync
    (:OnReady
      (bind _returnvalue (GameplayEventRouter|GetEventRouter "/Script/FortniteGame.FortGameState" (Components|GetOwner)))
      (bind (_asynctaskproxy _payload) (GameplayEventRouter|StartListeningToEvent _returnvalue "/Game/Sounds/Fort_Audio_Blueprints/EmoteMusicRegistrationComponent/GameplayEvent_OnPawnEmoteRegistration.GameplayEvent_OnPawnEmoteRegistration")
        (:OnEventReceived
          (if (.bRegister_5_8B89A8474843020216D86BADF79346CD (Utilities|Struct|BreakGameplayEventonPawnEmoteRegistration _payload))
            (CallFunction|OnRegister (Utilities|Struct|BreakGameplayEventonPawnEmoteRegistration _payload))
            (else
              (CallFunction|OnUnregister (Utilities|Struct|BreakGameplayEventonPawnEmoteRegistration _payload)))))))))

(fn ActivateEmoteDucker
  (if (and (> (Variables|Default|GetEmoteCount) 0) (not (Audio|Modulation|IsControlBusMixActive (Variables|Default|GetEmoterDuckerCBM))))
    (Audio|Modulation|ActivateControlBusMix (Variables|Default|GetEmoterDuckerCBM))))

(fn DeactivateEmoteDucker
  (if (and (<= (Variables|Default|GetEmoteCount) 0) (Audio|Modulation|IsControlBusMixActive (Variables|Default|GetEmoterDuckerCBM)))
    (Audio|Modulation|DeactivateControlBusMix (Variables|Default|GetEmoterDuckerCBM))))

(fn OnRegister
  (params (FortPawnRef :type Fort%20Pawn%20Object%20Reference))
  (bind _emotingpawns (Variables|Default|GetEmotingPawns))
  (Utilities|Array|AddUnique _emotingpawns FortPawnRef)
  (Variables|Default|SetEmoteCount (Utilities|Array|Length _emotingpawns))
  (CallFunction|ActivateEmoteDucker))

(fn OnUnregister
  (params (FortPawnRef :type Fort%20Pawn%20Object%20Reference))
  (bind _emotingpawns (Variables|Default|GetEmotingPawns))
  (if (Utilities|Array|IsNotEmpty _emotingpawns)
    (bind _returnvalue (Utilities|Array|RemoveItem _emotingpawns (Variables|GetFortPawnRef)))
    (if _returnvalue
      (Variables|Default|SetEmoteCount (Utilities|Array|Length (Variables|Default|GetEmotingPawns)))
      (CallFunction|DeactivateEmoteDucker))))
