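; Player pawn component that registers and unregisters emoting pawns for emote music through a gameplay event on the game state.
; Once the game state is ready, on machines running cosmetic events, it binds to the owner's emote start and stop dispatchers.
; Locally viewed pawns register immediately; others run a looping timer that registers them only while an audio listener is within range.
; Stopping clears that timer and unregisters, and a delta gate limits broadcasts to registration changes.
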(event EventBeginPlay
  (Class|FortAsyncActionGameStateReady|GameStateReadyAsync
    (:OnReady
      (Networking|CanExecuteCosmeticEvents
        (:True
          (bind _asfort_player_pawn (Utilities|Casting|CastToFortPlayerPawn (Components|GetOwner)))
          (EventDispatchers|AssignOnPawnStartedEmote _asfort_player_pawn (delegate Custom|OnPawnStartedEmote_Event))
          (EventDispatchers|BindEventtoOnPawnStoppedEmote _asfort_player_pawn (delegate Custom|OnPawnStoppedEmote_Event)))))))

(event Custom|OnPawnStartedEmote_Event
  (params (MontageItemDef :type Fort%20Item%20Definition%20Object%20Reference) (PawnEmoting :type Fort%20Pawn%20Object%20Reference))
  (if (Pawn|IsLocallyViewed PawnEmoting)
    (CallFunction|RegisterwithManager PawnEmoting true)
    (else
      (Variables|Default|SetPlayerPawn PawnEmoting)
      (CallFunction|InitRangeTimer))))

(event Custom|OnPawnStoppedEmote_Event
  (params (MontageItemDef :type Fort%20Item%20Definition%20Object%20Reference) (PawnEmoting :type Fort%20Pawn%20Object%20Reference))
  (bind _inrangetimerhandle (Variables|Default|GetInRangeTimerHandle))
  (if (Pawn|IsLocallyViewed PawnEmoting)
    _
    (elif (Utilities|Time|IsValidTimerHandle _inrangetimerhandle)
      (Utilities|Time|ClearandInvalidateTimerbyHandle _inrangetimerhandle)))
  (CallFunction|RegisterwithManager PawnEmoting))

(event Custom|RegisterWithManager
  (params (FortPawn :type Fort%20Pawn%20Object%20Reference) (bRegister :type Boolean))
  (Utilities|FlowControl|DeltaGateWildcard (Variables|Default|GetbRegisterState) bRegister
    (:Success
      (GameplayEvent|BroadcastEvent (Variables|Default|GetOnEmoteEventDescriptor) 0 (Utilities|Struct|MakeGameplayEventonPawnEmoteRegistration FortPawn bRegister) "/Script/FortniteGame.FortGameState"))))

(fn InitRangeTimer
  (bind _returnvalue (Utilities|Time|SetTimerbyEvent (EventDispatchers|CreateEvent :!event IsListenerInRange) 0.5 true))
  (Variables|Default|SetInRangeTimerHandle _returnvalue))

(fn IsListenerInRange
  (bind _playerpawn (Variables|Default|GetPlayerPawn)
    (:then
      (bind _returnvalue (Audio|AreAnyListenersWithinRange (Transformation|GetActorLocation _playerpawn) 1e+03))
      (CallFunction|RegisterwithManager _playerpawn _returnvalue))
    (:else)))
